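; Gameplay ability that makes the LilWhip AI bot throw a randomly chosen ice cream.
; It removes special variants, then rolls ScalableFloat chances in priority order (Spicy, Frozen, Guzzle, Slurp), falling back to Normal, and grants any special pick.
; It equips by tag, holds secondary fire for a random delay with a speech bubble, then releases and ends.
; CommitAbility runs in OnEndAbility rather than on activation, so failed or cancelled runs also commit.
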(event Ability|EventActivateAbility
  (bind _asfort_athena_aibot_controller (Utilities|Casting|CastToFortAthenaAIBotController (AI|GetAIController (Ability|GetAvatarActorfromActorInfo)))
    (:then
      (Variables|InternalData|SetAIBotController _asfort_athena_aibot_controller)
      (CallFunction|CleanInventory)
      (bind _output_get (Variables|Default|SetSelectedWeapon (CallFunction|SelectIcecreamWeaponTagbyEnum)))
      (if (not (Utilities|Enum|Equal(Enum) _output_get "NewEnumerator0"))
        (bind _returnvalue (Inventory|GiveItemtoInventoryOwner :InventoryOwner _asfort_athena_aibot_controller :ItemDefinition (Utilities|Select :Index (Variables|Default|GetSelectedWeapon) :NewEnumerator0 (Variables|Default|GetNormalIncecreamItemDef) :NewEnumerator1 (Variables|Default|GetSpicyIncecreamItemDef) :NewEnumerator2 (Variables|Default|GetFrozenIncecreamItemDef) :NewEnumerator3 (Variables|Default|GetGuzzleIncecreamItemDef) :NewEnumerator4 (Variables|Default|GetSlurpIncecreamItemDef)) :NumberToGive 1 :bNotifyPlayer false :ItemLevel -1 :PickupInstigatorHandle -1 :bUseItemPickupAnalyticEvent false :WeaponAmmoOverride -1))
        (Utilities|IsValid _returnvalue
          (:Is%20Valid
            _)
          (:Is%20Not%20Valid
            (Ability|EndAbility))))
      (bind _returnvalue_2 (Class|FortAthenaAIBotController|EquipWeaponbyTag _asfort_athena_aibot_controller (Utilities|Select :Index (Variables|Default|GetSelectedWeapon) :NewEnumerator0 (Variables|Default|GetNormalIcecreamTag) :NewEnumerator1 (Variables|Default|GetSpicyIcecreamTag) :NewEnumerator2 (Variables|Default|GetFrozenIcecreamTag) :NewEnumerator3 (Variables|Default|GetGuzzleIcecreamTag) :NewEnumerator4 (Variables|Default|GetSlurpIcecreamTag))))
      (if (Utilities|IsValid _returnvalue_2 :!class /Script/Engine.KismetSystemLibrary)
        (Utilities|FlowControl|Delay 0.5)
        (Class|FortAthenaAIBotController|StartSecondaryFire _asfort_athena_aibot_controller)
        (CallFunction|StartSpeechBubble)
        (Utilities|FlowControl|Delay (Math|Random|RandomFloatinRange 0.75 1.5))
        (Class|FortAthenaAIBotController|StopSecondaryFire _asfort_athena_aibot_controller true)
        (Utilities|FlowControl|Delay 0.6)
        (Ability|EndAbility)
        (else
          (Ability|EndAbility))))
    (:CastFailed
      (Ability|EndAbility))))

(event Ability|EventFailedToActivatePassiveAbility
  (params (ActorInfo :type Gameplay%20Ability%20Actor%20Info%20Structure)))

(event Ability|EventOnEndAbility
  (params (bWasCancelled :type Boolean))
  (|CallParentFunction bWasCancelled)
  (Ability|CommitAbility))

(fn CleanInventory
  (bind _aibotcontroller (Variables|InternalData|GetAIBotController))
  (Utilities|IsValid _aibotcontroller
    (:Is%20Valid
      (for _index (range 1 5)
        (Inventory|RemoveItemfromInventoryOwner _aibotcontroller (Utilities|Select :Index _index :Option%200 (Variables|Default|GetNormalIncecreamItemDef) :Option%201 (Variables|Default|GetSpicyIncecreamItemDef) :Option%202 (Variables|Default|GetFrozenIncecreamItemDef) :Option%203 (Variables|Default|GetGuzzleIncecreamItemDef) :Option%204 (Variables|Default|GetSlurpIncecreamItemDef))))
      (return))))

(fn SelectIcecreamWeaponTag
  (return (select (<= (Math|Random|RandomFloatinRange 0.0 1.0) (ScalableFloat|GetValueatLevel (Variables|Default|GetChanceForSpicy))) "NewEnumerator1" (select (<= (Math|Random|RandomFloatinRange 0.0 1.0) (ScalableFloat|GetValueatLevel (Variables|Default|GetChanceForFrozen))) "NewEnumerator2" "NewEnumerator0"))))

(fn SelectIcecreamWeaponTagByEnum
  (return (select (<= (Math|Random|RandomFloatinRange 0.0 1.0) (ScalableFloat|GetValueatLevel (Variables|Default|GetChanceForSpicy))) "NewEnumerator1" (select (<= (Math|Random|RandomFloatinRange 0.0 1.0) (ScalableFloat|GetValueatLevel (Variables|Default|GetChanceForFrozen))) "NewEnumerator2" (select (<= (Math|Random|RandomFloatinRange 0.0 1.0) (ScalableFloat|GetValueatLevel (Variables|Default|GetChanceForGuzzling))) "NewEnumerator3" (select (<= (Math|Random|RandomFloatinRange 0.0 1.0) (ScalableFloat|GetValueatLevel (Variables|Default|GetChanceForSlurp))) "NewEnumerator4" "NewEnumerator0"))))))

(fn StartSpeechBubble
  (bind _returnvalue (Ability|GetAvatarActorfromActorInfo))
  (bind _returnvalue_2 (Actor|GetComponentbyClass _returnvalue "/Script/FortniteGame.FortPawnComponent_SpeechBubble"))
  (Utilities|IsValid _returnvalue
    (:Is%20Valid
      (Utilities|IsValid _returnvalue_2
        (:Is%20Valid
          (SpeechBubble|StartSpeech _returnvalue_2 "(TokenStreamVersion=0,TagDictionary=((TagName=\"Athena.AI.SpeechBubbles.LilWhip.Throwing\")),QueryTokenStream=(0,1,1,1,0),UserDescription=\"\",AutoDescription=\" ANY( Athena.AI.SpeechBubbles.LilWhip.Throwing )\")"))))))
